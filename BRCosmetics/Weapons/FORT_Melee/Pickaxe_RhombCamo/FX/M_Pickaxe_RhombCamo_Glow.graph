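# Unlit, translucent, two-sided glow material for Niagara mesh particles on the RhombCamo pickaxe.
# Two stretched Perlin samples scroll in opposite U directions and combine via Max into streaky noise.
# That noise, masked by a V gradient, blends the ShapeColor and DropShadowColor parameters, with a powered noise term adding hot highlights.
# The same masked noise, multiplied by depth fade and particle alpha, forms an opacity-style term.
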Material /BRCosmetics/Weapons/FORT_Melee/Pickaxe_RhombCamo/FX/M_Pickaxe_RhombCamo_Glow
graph {
 properties {
  bUsedWithNiagaraMeshParticles=true
  blendMode=BLEND_Translucent
  pixelDepthOffsetMode=PDOM_Legacy
  shadingModel=MSM_Unlit
  twoSided=true
  usedShadingModels=Unlit
 }
 nodes [
  n1 = TextureCoordinate()
  n2 = Constant2Vector() {g=0.02, r=8}
  n3 = Time()
  n4 = Constant2Vector() {r=1.6}
  n5 = TextureSample(UVs=Add(A=Multiply(A=@n1, B=@n2), B=Multiply(A=@n3, B=@n4))) {samplerType=SAMPLERTYPE_LinearColor, texture=&/T_Pickaxe_RhombCamo_Perlin}
  n6 = TextureCoordinate()
  n7 = Constant2Vector() {g=0.02, r=3}
  n8 = Time()
  n9 = Constant2Vector() {r=-2.5}
  n10 = TextureSample(UVs=Add(A=Multiply(A=@n6, B=@n7), B=Multiply(A=@n8, B=@n9))) {samplerType=SAMPLERTYPE_LinearColor, texture=&T_Pickaxe_RhombCamo_Perlin}
  n11 = Max(A=@n5.R, B=@n10.G)
  n12 = VectorParameter() {group=RhombCamoVisor, parameterName=DropShadowColor, sortPriority=40, defaultValue=<rgba 10 6.6346 2.1993 1>}
  n13 = Multiply(A=ComponentMask(None=TextureCoordinate()) {g=true}, B=@n11)
  n14 = VectorParameter() {group=RhombCamoVisor, parameterName=ShapeColor, sortPriority=40, defaultValue=<rgba 1 0.0084 0 1>}
  Add(
   A=Multiply(A=Power(Base=@n11) {constExponent=7}, B=Multiply(A=@n12.RGB) {constB=3})
   B=LinearInterpolate(A=@n14.RGB, B=@n12.RGB, Alpha=Power(Base=@n13) {constExponent=2.7})
  )
  Multiply(A=@n5.R, B=@n10.G)
  Multiply(A=@n14.RGB) {constB=5}
  Multiply(A=Multiply(A=Power(Base=@n13) {constExponent=1.45}, B=DepthFade() {fadeDistanceDefault=10}), B=ParticleColor().A)
 ]
}
