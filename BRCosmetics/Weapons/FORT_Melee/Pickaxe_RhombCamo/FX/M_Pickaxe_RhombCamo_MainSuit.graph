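# Translucent, unlit, two-sided effect material for Niagara mesh particles in the RhombCamo pickaxe FX, tinted by a red-by-default ShapeColor parameter.
# Scrolling hard-edged bands, thinning along the masked UV axis, are modulated by a ScanLines function and a fade, then added to a soft end gradient.
# That mask is multiplied by particle alpha and the maximum of two time-driven Perlin texture samples moving in opposite directions.
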Material /BRCosmetics/Weapons/FORT_Melee/Pickaxe_RhombCamo/FX/M_Pickaxe_RhombCamo_MainSuit
graph {
 properties {
  bUsedWithNiagaraMeshParticles=true
  blendMode=BLEND_Translucent
  pixelDepthOffsetMode=PDOM_Legacy
  shadingModel=MSM_Unlit
  twoSided=true
  usedShadingModels=Unlit
 }
 nodes [
  n1 = TextureCoordinate()
  n2 = Multiply(A=Time(), B=Constant2Vector() {r=1.6})
  Add(A=Multiply(A=@n1, B=Constant2Vector() {r=8}), B=@n2)
  n3 = Multiply(A=Time(), B=Constant2Vector() {r=-2.5})
  Add(A=Multiply(A=@n1, B=Constant2Vector() {r=3}), B=@n3)
  n4 = TextureCoordinate()
  ComponentMask(None=@n4) {g=true}
  n5 = MaterialFunctionCall!(
   ScanLine_Tiling_1=ScalarParameter() {defaultValue=10, group=RhombCamoVisor, parameterName=ScanLineTiling1, sortPriority=300}
   Scanline_PannerSpeed_1=ScalarParameter() {defaultValue=-1, group=RhombCamoVisor, parameterName=ScanlineSpeed1, sortPriority=301}
   ScanLine_Tiling=ScalarParameter() {defaultValue=110, group=RhombCamoVisor, parameterName=ScanLineTiling2, sortPriority=302}
   Scanline_PannerSpeed
   UVs
  ) {materialFunction=/Game/Weapons/WeaponSkins/MaterialFunction/ScanLines}
  n6 = ScalarParameter() {defaultValue=0.5, group=RhombCamoVisor, parameterName=ScanLineStrength, sortPriority=299}
  n7 = StaticComponentMaskParameter(None=TextureCoordinate()) {defaultR=true, parameterName=UVMask}
  n8 = LinearInterpolate(A=ScalarParameter() {defaultValue=0.8, parameterName=LineThickness}, Alpha=Power(Base=@n7) {constExponent=1}) {constA=0.8, constB=0.01}
  n9 = Multiply(A=@n7) {constB=4}
  n10 = Multiply(A=Time(), B=ScalarParameter() {defaultValue=-1, parameterName=ScrollSpeed}) {constB=-1}
  n11 = SmoothStep(Min=@n8, Max=Subtract(A=@n8) {constB=0.01}, Value=Frac(None=Add(A=@n9, B=@n10))) {constMin=0.5}
  n12 = OneMinus(None=StaticComponentMaskParameter(None=@n4) {defaultR=true, parameterName=UVMask})
  n13 = Saturate(None=Multiply(A=@n12) {constB=2})
  n14 = Add(
   A=Multiply(A=LinearInterpolate(B=@n5.Result, Alpha=@n6) {constA=1}, B=Multiply(A=@n11, B=@n13))
   B=Multiply(A=Power(Base=@n12) {constExponent=3}) {constB=0.45}
  ) {constB=0.1}
  Multiply(
   A=Multiply(A=ParticleColor().A, B=Saturate(None=@n14))
   B=Max(
    A=TextureSample(UVs=@n2) {samplerType=SAMPLERTYPE_LinearColor, texture=&/T_Pickaxe_RhombCamo_Perlin}.R
    B=TextureSample(UVs=@n3) {samplerType=SAMPLERTYPE_LinearColor, texture=&T_Pickaxe_RhombCamo_Perlin}.R
   )
  )
  Multiply(A=VectorParameter() {group=RhombCamoVisor, parameterName=ShapeColor, sortPriority=40, defaultValue=<rgba 1 0.0084 0 1>}.RGB) {constB=0.9}
 ]
}
