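# Translucent unlit Niagara sprite material for the RhombCamo pickaxe impact effect, drawing a card-suit symbol from an SDF Shape texture, clubs by default.
# An inverted, smoothstepped SDF sample forms the mask, while a center-scaled sample plus time drives animated contour bands that follow the shape.
# ScanLines are blended over that pattern by a strength parameter, then multiplied by particle alpha.
# Particle color feeds the shared emissive adjustments function.
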Material /BRCosmetics/Weapons/FORT_Melee/Pickaxe_RhombCamo/FX/M_Pickaxe_RhombCamo_SuitImpact
graph {
 properties {bUsedWithNiagaraSprites=true, blendMode=BLEND_Translucent, pixelDepthOffsetMode=PDOM_Legacy, shadingModel=MSM_Unlit, usedShadingModels=Unlit}
 nodes [
  Constant() {r=0.47}
  n1 = ParticleColor()
  MaterialFunctionCall(Emissive=@n1.RGB) {materialFunction=&MF_FX_EmissiveAdjustments}
  n2 = TextureSampleParameter2D() {
   parameterName=Shape
   samplerType=SAMPLERTYPE_LinearColor
   texture=/BRCosmetics/Weapons/FORT_Melee/Pickaxe_RhombCamo/FX/SuitsTextures/T_Pickaxe_RhombCamo_Clubs_SDF
  }
  n3 = Power(Base=SmoothStep(Value=OneMinus(None=@n2.R)) {constMin=0.3})
  n4 = TextureCoordinate()
  n5 = Constant2Vector() {g=2, r=2}
  n6 = TextureSampleParameter2D(UVs=MaterialFunctionCall(UVs=@n4, Texture Scale=@n5) {materialFunction=&ScaleUVsByCenter}.UVs) {
   parameterName=Shape
   samplerType=SAMPLERTYPE_LinearColor
   texture=&T_Pickaxe_RhombCamo_Clubs_SDF
  }
  n7 = LinearInterpolate(Alpha=Power(Base=@n6.R) {constExponent=0.35}) {constA=0.07}
  n8 = Time()
  n9 = Frac(None=Add(A=Multiply(A=@n6.R) {constB=12}, B=Multiply(A=@n8) {constB=-4}))
  n10 = SmoothStep(Min=@n7, Max=Add(A=@n7) {constB=0.02}, Value=Clamp(None=@n9) {minDefault=0.4}) {constMax=0.52, constMin=0.5}
  n11 = MaterialFunctionCall!(
   ScanLine_Tiling_1=ScalarParameter() {defaultValue=5, group=RhombCamoVisor, parameterName=ScanLineTiling1, sortPriority=300}
   Scanline_PannerSpeed_1=ScalarParameter() {defaultValue=-1, group=RhombCamoVisor, parameterName=ScanlineSpeed1, sortPriority=301}
   ScanLine_Tiling=ScalarParameter() {defaultValue=50, group=RhombCamoVisor, parameterName=ScanLineTiling2, sortPriority=302}
   Scanline_PannerSpeed
   UVs
  ) {materialFunction=/Game/Weapons/WeaponSkins/MaterialFunction/ScanLines}
  n12 = ScalarParameter() {defaultValue=0.3, group=RhombCamoVisor, parameterName=ScanLineStrength, sortPriority=299}
  Multiply(A=@n1.A, B=Multiply(A=Multiply(A=@n3, B=@n10), B=LinearInterpolate(B=@n11.Result, Alpha=@n12) {constA=1}))
  Multiply(A=@n4) {constB=0.5}
  SmoothStep(Value=@n9) {constMax=0.06, constMin=0.07}
  SmoothStep() {constMax=0.06, constMin=0.07}
  Subtract(A=@n4, B=@n5) {constB=-0.432}
  TextureCoordinate()
 ]
}
