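# Translucent unlit Niagara sprite material for the RhombCamo pickaxe's swing suit effect, drawing a glitching, flickering outline of a suit shape, a diamond by default.
# Two scrolling glitch samples distort the Shape SDF's UVs, and paired SmoothSteps cut a thin band from it.
# Scanlines, particle randomness, particle alpha and depth fade modulate the band, while particle color drives emissive.
# Position-offset, panning Perlin noise provides random flickering.
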Material /BRCosmetics/Weapons/FORT_Melee/Pickaxe_RhombCamo/FX/M_Pickaxe_RhombCamo_SwingSuits
graph {
 properties {
  bUsedWithNiagaraSprites=true
  blendMode=BLEND_Translucent
  pixelDepthOffsetMode=PDOM_Legacy
  shadingModel=MSM_Unlit
  twoSided=true
  usedShadingModels=Unlit
 }
 nodes [
  MaterialFunctionCall(Emissive=ParticleColor().RGB) {materialFunction=&MF_FX_EmissiveAdjustments}
  n1 = Multiply(A=Constant2Vector() {g=1}, B=NamedRerouteUsage())
  n2 = Multiply(A=Add(A=TextureCoordinate(), B=@n1), B=Constant2Vector() {g=2, r=1})
  n3 = Multiply(A=Time(), B=Constant2Vector() {g=0.5})
  n4 = Multiply(A=Add(A=TextureCoordinate(), B=@n1), B=Constant2Vector() {g=0.7, r=1})
  n5 = Multiply(A=Time(), B=Constant2Vector() {g=-0.6, r=0.2})
  n6 = Max(
   A=TextureSample(UVs=Add(A=@n2, B=@n3)) {samplerType=SAMPLERTYPE_LinearGrayscale, texture=&/T_RhombCamo_Glitch}.R
   B=TextureSample(UVs=Add(A=@n4, B=@n5)) {samplerType=SAMPLERTYPE_LinearGrayscale, texture=&T_RhombCamo_Glitch}.R
  )
  n7 = TextureCoordinate()
  n8 = Multiply(A=ComponentMask(None=OneMinus(None=@n7)) {r=true}) {constB=2}
  n9 = NamedRerouteUsage()
  n10 = ParticleRandom()
  n11 = Multiply(A=Multiply(A=Subtract(A=@n8), B=Constant2Vector() {r=0.17}) {constB=0.3}, B=Multiply(A=Multiply(A=@n9, B=@n10)) {constB=2})
  n12 = TextureSampleParameter2D(UVs=Add(A=TextureCoordinate(), B=Multiply(A=@n6, B=@n11))) {
   parameterName=Shape
   samplerType=SAMPLERTYPE_LinearColor
   texture=/BRCosmetics/Weapons/FORT_Melee/Pickaxe_RhombCamo/FX/SuitsTextures/T_Pickaxe_RhombCamo_Diamond_SDF
  }
  n13 = OneMinus(None=@n12.R)
  n14 = ScalarParameter() {defaultValue=5, group=RhombCamoVisor, parameterName=ScanLineTiling1, sortPriority=300}
  n15 = ScalarParameter() {defaultValue=-1, group=RhombCamoVisor, parameterName=ScanlineSpeed1, sortPriority=301}
  n16 = ScalarParameter() {defaultValue=50, group=RhombCamoVisor, parameterName=ScanLineTiling2, sortPriority=302}
  n17 = Multiply(
   A=Subtract(A=SmoothStep(Value=@n13) {constMax=0.9, constMin=0.88}, B=SmoothStep(Value=@n13) {constMax=0.91, constMin=0.9})
   B=LinearInterpolate(
    B=MaterialFunctionCall!(ScanLine_Tiling_1=@n14, Scanline_PannerSpeed_1=@n15, ScanLine_Tiling=@n16, Scanline_PannerSpeed, UVs) {materialFunction=/Game/Weapons/WeaponSkins/MaterialFunction/ScanLines}.Result
    Alpha=ScalarParameter() {defaultValue=0.5, group=RhombCamoVisor, parameterName=ScanLineStrength, sortPriority=299}
   ) {constA=1}
  )
  n18 = Multiply(A=NamedRerouteUsage(), B=ParticleRandom())
  Multiply(A=Multiply(A=Multiply(A=@n17, B=@n18), B=ParticleColor().A), B=DepthFade() {fadeDistanceDefault=15})
  n19 = Time()
  n20 = LinearInterpolate(A=ObjectPositionWS(), B=ParticlePositionWS(), Alpha=ScalarParameter() {parameterName=RandomPerParticle})
  n21 = NamedRerouteDeclaration(None=Multiply(A=ComponentMask(None=@n20) {g=true, r=true}) {constB=0.3}) {name=PositionRandom, nodeColor=<rgba 1 0 0.7969 1>}
  n22 = TextureSample(UVs=Multiply(A=Add(A=@n19, B=@n21), B=Constant2Vector() {g=0.4, r=3.5})) {samplerType=SAMPLERTYPE_LinearColor, texture=&/T_Pickaxe_RhombCamo_Perlin}
  n23 = Time()
  n24 = TextureSample(UVs=Multiply(A=Add(A=@n23, B=@n21), B=Constant2Vector() {g=-0.6, r=-5})) {samplerType=SAMPLERTYPE_LinearColor, texture=&T_Pickaxe_RhombCamo_Perlin}
  NamedRerouteDeclaration(None=Power(Base=Max(A=@n22.R, B=@n24.R)) {constExponent=3}) {name=RandomFlickering, nodeColor=<rgba 1 0.3281 0 1>}
 ]
}
