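# Idle Niagara effect for the RhombCamo pickaxe, with a sprite glow, staff-head meshes, a suit hologram and its drop shadow.
# Four emitters burst-spawn particles that outlive their lifetime and follow locations sampled from the user skeletal mesh.
# User ShapeID picks the hologram and shadow mesh, indexing arrays for hologram color and shadow alpha.
# Time of day scales staff glow, hologram and shadow alpha, with a ShapeID-dependent morning and day hologram boost.
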NiagaraSystem /BRCosmetics/Weapons/FORT_Melee/Pickaxe_RhombCamo/FX/NS_Pickaxe_RhombCamo_Idle
dependencies {
 usedDataInterfaces [
  &NiagaraDataInterfaceArrayColor
  &NiagaraDataInterfaceMeshRendererInfo
  &NiagaraDataInterfaceArrayFloat
 ]
 usedDynamicInputs []
 usedModules [
  &SystemState
  &self:SystemSpawnScript.SystemScriptSource.SystemScriptGraph.NiagaraNodeAssignment_0.SetVariables_#1
 ]
}
emitters {
 Glow {
  emitterUpdateScript [
   &EmitterState(Life Cycle Mode=::NewEnumerator1, Loop Behavior=::NewEnumerator1, Loop Duration=1, UseLoopDelay=false)
   &SpawnBurst_Instantaneous(Use Spawn Probability=false, UseLoopCountLimit=false)
  ]
  particleSpawnScript [&InitializeParticle(
    Color=$System.GlowColor
    Lifetime=5
    Sprite Size Mode=::NewEnumerator3
    Uniform Sprite Size=110
    UsePositionOffset=false
    Write Material Random=true
   )]
  particleUpdateScript [
   &ParticleState(Kill Particles When Lifetime Has Elapsed=false)
   &SkeletalMeshLocation(
    MaskSpawnGroup=true
    Mesh Sampling Type=::NewEnumerator5
    Sampled Position Offset=<0.5 0 0.5>
    Skeletal Mesh=$User.SkeletalMesh
    Socket Sampling Mode=::NewEnumerator5
   )
   &CameraOffset(Camera Offset Amount=15)
  ]
  rendererClasses [&NiagaraSpriteRendererProperties]
  renderers [{rendererClass=&NiagaraSpriteRendererProperties}]
 }
 StaffHeadEffects {
  emitterUpdateScript [
   &EmitterState(Life Cycle Mode=::NewEnumerator1, Loop Behavior=::NewEnumerator1, Loop Duration=1, UseLoopDelay=false)
   &SpawnBurst_Instantaneous(Use Spawn Probability=false, UseLoopCountLimit=false)
  ]
  particleSpawnScript [
   &InitializeParticle(
    Color=<rgba 1 1 1 0.8>
    Lifetime=5
    Sprite Size Mode=::NewEnumerator3
    Uniform Sprite Size=120
    UsePositionOffset=false
    Write Material Random=true
   )
   /SetVariables(Particles.GlowColor=<rgba 1 1 1 0.25>) {isSetParameters=true}
  ]
  particleUpdateScript [
   &ParticleState(Kill Particles When Lifetime Has Elapsed=false)
   &SkeletalMeshLocation(Bone Sampling Mode=::NewEnumerator4, MaskSpawnGroup=true, Sampled Position Offset=<0.2 0 0>, Skeletal Mesh=$User.SkeletalMesh)
   /SetVariables(Particles.GlowColor=&MultiplyLinearColor(
     Linear Color=<rgba 1 1 1 0.25>
     Scale Factor=&MakeVector4_FromFloat(Float=&FNTimeOfDayScalar(Night Scale=0.35, Morning Scale=1, Evening Scale=0.55))
    )) {isSetParameters=true}
  ]
  rendererClasses [&NiagaraMeshRendererProperties]
  renderers [
   {rendererClass=&NiagaraMeshRendererProperties, rendererIndex=0}
   {rendererClass=&NiagaraMeshRendererProperties, rendererIndex=1}
   {rendererClass=&NiagaraMeshRendererProperties, rendererIndex=2}
  ]
 }
 SuitHologram {
  emitterUpdateScript [
   &EmitterState(Life Cycle Mode=::NewEnumerator1, Loop Behavior=::NewEnumerator1, Loop Duration=1, UseLoopDelay=false)
   &SpawnBurst_Instantaneous(Use Spawn Probability=false, UseLoopCountLimit=false)
  ]
  particleSpawnScript [&InitializeParticle(
    Color=&SelectLinearColorFromArray(
     Array Sampling Mode=::NewEnumerator1
     Direct Array Index=&MakeIntFromFloat(Float=$User.ShapeID)
     Color Selection Array=DataInterface() {propertyValues {ColorData [<rgba 0.8281 0.8281 0.8281 1>, <rgba 0.8281 0.8281 0.8281 1>, <rgba 0.8281 0.8281 0.8281 1>, <rgba 2 2 2 1>, <rgba 2 2 2 1>]}}
     Random Seed=?
    )
    Lifetime=5
    Mesh Renderer Array Visibility Mode=::NewEnumerator1
    Mesh Renderer Info=DataInterface() {propertyValues {MeshRenderer=&self:SuitHologram_2.NiagaraMeshRendererProperties_2}}
    Mesh Renderer Mesh Index:NiagaraInt32=&MakeIntFromFloat(Float=$User.ShapeID, Conversion Mode=::NewEnumerator3)
    Sprite Size Mode=::NewEnumerator3
    Uniform Sprite Size=120
    UsePositionOffset=false
    Write Material Random=true
   )]
  particleUpdateScript [
   &ParticleState(Kill Particles When Lifetime Has Elapsed=false)
   &SkeletalMeshLocation(Bone Sampling Mode=::NewEnumerator4, MaskSpawnGroup=true, Sampled Position Offset=<-3 0 0>, Skeletal Mesh=$User.SkeletalMesh)
   &ScaleColor(
    Scale Alpha=&FN_Select_Float_from_Time_Of_Day(
     Morning=HlslExpression() {hlslExpression="User.ShapeID < 3 ? 1 : 2.8"}
     Day=HlslExpression() {hlslExpression="User.ShapeID < 3 ? 1 : 2.8"}
     Evening=1
     Night=1
    )
    ScaleA=true
    ScaleRGB=true
   )
  ]
  rendererClasses [&NiagaraMeshRendererProperties]
  renderers [{rendererClass=&NiagaraMeshRendererProperties, rendererIndex=0}]
 }
 SuitHologramDropShadow {
  emitterUpdateScript [
   &EmitterState(Life Cycle Mode=::NewEnumerator1, Loop Behavior=::NewEnumerator1, Loop Duration=1.4, UseLoopDelay=false)
   &SpawnBurst_Instantaneous(Use Spawn Probability=false, UseLoopCountLimit=false)
  ]
  particleSpawnScript [&InitializeParticle(
    Color=&MakeLinearColor(
     R:NiagaraFloat=1
     G:NiagaraFloat=1
     B:NiagaraFloat=1
     A:NiagaraFloat=&SelectFloatFromArray(
      Direct Array Index=&MakeIntFromFloat(Float=$User.ShapeID)
      Float Selection Array=DataInterface() {propertyValues {FloatData [1,1,1.2,1.1,1.1]}}
     )
    )
    Lifetime=5
    Mesh Renderer Array Visibility Mode=::NewEnumerator1
    Mesh Renderer Info=DataInterface() {propertyValues {MeshRenderer=&self:SuitHologramDropShadow_3.NiagaraMeshRendererProperties_2}}
    Mesh Renderer Mesh Index:NiagaraInt32=&MakeIntFromFloat(Float=$User.ShapeID, Conversion Mode=::NewEnumerator3)
    Sprite Size Mode=::NewEnumerator3
    Uniform Sprite Size=120
    UsePositionOffset=false
    Write Material Random=true
   )]
  particleUpdateScript [
   &ParticleState(Kill Particles When Lifetime Has Elapsed=false)
   &SkeletalMeshLocation(Bone Sampling Mode=::NewEnumerator4, MaskSpawnGroup=true, Sampled Position Offset=<-3.5 -1.5 0>, Skeletal Mesh=$User.SkeletalMesh)
   &ScaleColor(
    Scale Alpha=&FN_Select_Float_from_Time_Of_Day(Morning=0.7, Day=1, Evening=1, Night=0.5)
    ScaleA=true
    ScaleRGB=true
   )
  ]
  rendererClasses [&NiagaraMeshRendererProperties]
  renderers [{rendererClass=&NiagaraMeshRendererProperties, rendererIndex=0}]
 }
}
summary {
 systemName=NS_Pickaxe_RhombCamo_Idle
 emitters [
  {emitterName=StaffHeadEffects, rendererClasses [&NiagaraMeshRendererProperties]}
  {emitterName=Glow, rendererClasses [&NiagaraSpriteRendererProperties]}
  {emitterName=SuitHologram, rendererClasses [&NiagaraMeshRendererProperties]}
  {emitterName=SuitHologramDropShadow, rendererClasses [&NiagaraMeshRendererProperties]}
 ]
 userVariables [
  User.DropShadowColor {type=LinearColor, description="", defaultValue {struct=LinearColor, value=<rgba 1 0.1674 0 1>}}
  User.ShapeColor {type=LinearColor, description="", defaultValue {struct=LinearColor, value=<rgba 1 0 0 1>}}
  User.ShapeID {type=NiagaraFloat, defaultValue=0, description=""}
  User.SkeletalMesh {
   type=NiagaraDataInterfaceSkeletalMesh
   defaultValue {
    struct=NiagaraExt_VariableValue_DataInterface
    value {dataInterface=&self:NiagaraDataInterfaceSkeletalMesh_0}
   }
  }
  User.SuitTexture {
   type=Texture
   description=""
   defaultValue {
    struct=NiagaraExt_VariableValue_Object
    value {object=/Game/Characters/Player/Male/Medium/Bodies/M_MED_RhombCamo/Textures/T_RhomboCamo_FaceAcc_SDF, objectClass=/Script/Engine.Texture}
   }
  }
 ]
}
