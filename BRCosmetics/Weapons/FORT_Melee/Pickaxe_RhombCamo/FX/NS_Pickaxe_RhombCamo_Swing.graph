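# Swing effect for the RhombCamo pickaxe, drawing ribbon trails and flickering card-suit sprites from the weapon's skeletal mesh.
# BlackTrail and SuitColorTrail are short-lived anim trails between two bones or sockets, one dark, one tinted by ShapeColor.
# Suits spawns sprites by distance moved, tinted by ShapeColor and sized by ShapeType through array lookups, with time-of-day alpha scaling.
# A user ShapeTexture parameter defaults to a diamond SDF.
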NiagaraSystem /BRCosmetics/Weapons/FORT_Melee/Pickaxe_RhombCamo/FX/NS_Pickaxe_RhombCamo_Swing
dependencies {
 usedDataInterfaces [&NiagaraDataInterfaceArrayFloat, &NiagaraDataInterfaceCurve, &NiagaraDataInterfaceColorCurve]
 usedDynamicInputs []
 usedModules [&SystemState]
}
emitters {
 BlackTrail {
  emitterSpawnScript [&FN_RibbonIDOffsetUtility()]
  emitterUpdateScript [
   &EmitterState()
   &FN_SpawnAnimTrail(Number of Anim Trails=2)
  ]
  particleSpawnScript [
   &FN_InitializeAnimTrailBoneIndecies(First Bone or Socket Index=1, Second Bone or Socket Index=2)
   &FN_InitializeAnimTrail(AnimTrailWidth=1.2, Color=<rgba 0.0417 0.0334 0.0338 1>, Lifetime=0.096)
   &AddVelocity(Velocity=<0 400 0>) {enabled=false}
  ]
  particleUpdateScript [
   &SolveForcesAndVelocity(Clamp Velocity=false, Limit Acceleration=false)
   &FN_UpdateAnimTrail(Locked Particle RibbonWidth=1, SkeletalMesh=$User.SkeletalMesh) {module=FN_UpdateAnimTrail001}
   &ScaleColor(
    Scale Alpha=&FloatFromCurve(
     FloatCurve=DataInterface() {propertyValues {Curve {keys [
         {interpMode=RCIM_Cubic, tangentMode=RCTM_Break, time=0, value=1}
         {arriveTangent=-1.5656, interpMode=RCIM_Cubic, leaveTangent=-3.0266, tangentMode=RCTM_Break, time=0.4, value=0.7}
         {arriveTangent=-0.0937, interpMode=RCIM_Cubic, tangentMode=RCTM_Break, time=0.8, value=0}
        ]}}}
     Scale Curve=0.7
    )
    ScaleA=true
    ScaleRGB=true
   )
  ]
  rendererClasses [&NiagaraRibbonRendererProperties]
  renderers [
   {rendererClass=&NiagaraRibbonRendererProperties}
   {rendererClass=&NiagaraRibbonRendererProperties, rendererIndex=1}
  ]
 }
 SuitColorTrail {
  emitterSpawnScript [&FN_RibbonIDOffsetUtility()]
  emitterUpdateScript [
   &EmitterState()
   &FN_SpawnAnimTrail(Number of Anim Trails=2)
  ]
  particleSpawnScript [
   &FN_InitializeAnimTrailBoneIndecies(First Bone or Socket Index=1, Second Bone or Socket Index=2)
   &FN_InitializeAnimTrail(AnimTrailWidth=2, Color=$User.ShapeColor, Lifetime=0.2)
   &AddVelocity(Velocity=<0 400 0>) {enabled=false}
  ]
  particleUpdateScript [
   &SolveForcesAndVelocity(Clamp Velocity=false, Limit Acceleration=false)
   &FN_UpdateAnimTrail(Locked Particle RibbonWidth=1, SkeletalMesh=$User.SkeletalMesh) {module=FN_UpdateAnimTrail001}
   &ScaleColor(
    Scale Alpha=&FloatFromCurve(FloatCurve=DataInterface() {propertyValues {Curve {keys [<at 0 1>, {interpMode=RCIM_Cubic, leaveTangent=-2.5238, tangentMode=RCTM_Break, time=0.2, value=0.75}, <at 0.9 0>]}}})
    Scale RGB=&VectorFromFloat(Value=&FloatFromCurve(
      FloatCurve=DataInterface() {propertyValues {Curve {keys [{interpMode=RCIM_Cubic, leaveTangent=-2.5418, tangentMode=RCTM_Break, time=0, value=1}, <at 1 0>]}}}
      Scale Curve=5
     ))
    ScaleA=true
    ScaleRGB=true
   )
   &ScaleRibbonWidth(Ribbon Width Scale=&FloatFromCurve(FloatCurve=DataInterface() {propertyValues {Curve {keys [
         {interpMode=RCIM_Cubic, leaveTangent=-2.5402, tangentMode=RCTM_Break, time=0, value=1}
         {arriveTangent=0.0059, interpMode=RCIM_Cubic, tangentMode=RCTM_Break, time=1, value=0.2}
        ]}}}))
  ]
  rendererClasses [&NiagaraRibbonRendererProperties]
  renderers [
   {rendererClass=&NiagaraRibbonRendererProperties}
   {rendererClass=&NiagaraRibbonRendererProperties, rendererIndex=1}
  ]
 }
 Suits {
  emitterUpdateScript [
   &EmitterState()
   &SpawnPerUnit(
    Module.Use Max Movement Threshold=true
    Module.Use Movement Tolerance=false
    Spawn Probability=1
    Spawn Spacing=45
    Use Spawn Probability=true
   )
  ]
  particleSpawnScript [
   &InitializeParticle(
    Color=&HueShiftLinearColor(Hue=0, Saturation=1, Value=&RandomRangeFloat(Minimum=0.8, Maximum=1), Linear Color=$User.ShapeColor)
    Lifetime Max=0.4
    Lifetime Min=0.2
    Lifetime Mode=::NewEnumerator1
    Sprite Size Mode=::NewEnumerator1
    Uniform Sprite Size Max=&RandomRangeFloat(
     Minimum=&SelectFloatFromArray(
      Direct Array Index=&MakeIntFromFloat(Float=$User.ShapeType)
      Float Selection Array=DataInterface() {propertyValues {FloatData [140,110,100]}}
     )
     Maximum=&SelectFloatFromArray(
      Direct Array Index=&MakeIntFromFloat(Float=$User.ShapeType)
      Float Selection Array=DataInterface() {propertyValues {FloatData [160,130,150]}}
     )
    )
    Uniform Sprite Size Min=&SelectFloatFromArray(
     Direct Array Index=&MakeIntFromFloat(Float=$User.ShapeType)
     Float Selection Array=DataInterface() {propertyValues {FloatData [120,90,60]}}
    )
    UsePositionOffset=false
    Write Material Random=true
   )
   &SkeletalMeshLocation(MaskSpawnGroup=false, Mesh Sampling Type=::NewEnumerator5, Skeletal Mesh=$User.SkeletalMesh, Socket Sampling Mode=::NewEnumerator5)
   &InheritVelocity(Inherited Velocity Amount Scale=<0.1 0.1 0.1>, Speed Limit=true, UseSpeedThreshold=false)
   /SetVariables(Particles.SpriteFacing=$Particles.SkeletalMeshLocation.SampledVelocity) {enabled=false, isSetParameters=true}
  ]
  particleUpdateScript [
   &ParticleState()
   &ScaleSpriteSize(Uniform Curve Sprite Scale=DataInterface() {propertyValues {Curve {keys [
        {interpMode=RCIM_Cubic, leaveTangent=-0.8258, tangentMode=RCTM_Break, time=0, value=1}
        {arriveTangent=-0.1527, interpMode=RCIM_Cubic, leaveTangent=0.9374, tangentMode=RCTM_Break, time=1, value=0.5}
       ]}}}) {enabled=false}
   &ScaleSpriteSize(Uniform Curve Sprite Scale=DataInterface() {propertyValues {Curve {keys [
        {interpMode=RCIM_Cubic, leaveTangent=1.206, tangentMode=RCTM_Break, time=0, value=0.85}
        <at 0.2 1>
        <at 0.9 1>
        {arriveTangent=-0.3762, interpMode=RCIM_Cubic, leaveTangent=0.9374, tangentMode=RCTM_Break, time=1, value=0.9}
       ]}}}) {module=ScaleSpriteSize001}
   &ScaleColor(
    Scale Alpha=&FloatFromCurve(
     FloatCurve=DataInterface() {propertyValues {Curve {keys [
         <at 0 0.55>
         {arriveTangentWeight=0.0333, leaveTangentWeight=0.0667, time=0.1, value=1}
         <at 0.2 1>
         <at 0.3 0.2>
         <at 0.4 1>
         <at 0.5 1>
         <at 0.6 0.15>
         <at 0.7 1>
         <at 0.9 1>
         <at 1 0.35>
        ]}}}
     Scale Curve=&FN_Select_Float_from_Time_Of_Day(Morning=2, Day=2, Evening=0.9, Night=0.6)
    )
    Scale RGB=&VectorFromFloat(Value=&FloatFromCurve(FloatCurve=DataInterface() {propertyValues {Curve {keys [<at 0 0.35>,<at 0.1 15>,<at 0.15 2.5>,<at 0.4 1>,<at 1 0.5>]}}}))
    ScaleA=true
    ScaleRGB=true
   )
   &SolveForcesAndVelocity(Clamp Velocity=false, Limit Acceleration=false) {enabled=false}
  ]
  rendererClasses [&NiagaraSpriteRendererProperties]
  renderers [{rendererClass=&NiagaraSpriteRendererProperties}]
 }
}
summary {
 systemName=NS_Pickaxe_RhombCamo_Swing
 emitters [
  {emitterName=Suits, rendererClasses [&NiagaraSpriteRendererProperties]}
  {emitterName=BlackTrail, rendererClasses [&NiagaraRibbonRendererProperties]}
  {emitterName=SuitColorTrail, rendererClasses [&NiagaraRibbonRendererProperties]}
 ]
 userVariables [
  User.RibbonIDOffset {type=NiagaraInt32}
  User.ShapeColor {type=LinearColor, description="", defaultValue {struct=LinearColor, value=<rgba 1 0.0084 0 1>}}
  User.ShapeTexture {
   type=Texture
   description=""
   defaultValue {
    struct=NiagaraExt_VariableValue_Object
    value {
     object=/BRCosmetics/Weapons/FORT_Melee/Pickaxe_RhombCamo/FX/SuitsTextures/T_Pickaxe_RhombCamo_Diamond_SDF
     objectClass=/Script/Engine.Texture
    }
   }
  }
  User.ShapeType {type=NiagaraFloat, defaultValue=0, description=""}
  User.SkeletalMesh {
   type=NiagaraDataInterfaceSkeletalMesh
   defaultValue {
    struct=NiagaraExt_VariableValue_DataInterface
    value {dataInterface=&self:NiagaraDataInterfaceSkeletalMesh_0}
   }
  }
 ]
}
